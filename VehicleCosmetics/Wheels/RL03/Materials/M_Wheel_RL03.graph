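# Masked, lit material for the RL03 vehicle cosmetic wheel, adding an animated emissive effect to part of the rim.
# MF_WheelParent_Simple supplies the base attributes and rim mask; green channel times inverted vertex-color red selects the effect area.
# On UV1, two panning caustics samples combine with scrolling circular-gradient bands, then get squared, saturated and blended into Base Color and Emissive.
# The result passes through vehicle damage and de-rez transition functions.
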Material /VehicleCosmetics/Wheels/RL03/Materials/M_Wheel_RL03
graph {
 properties {
  bAutomaticallySetUsageInEditor=false
  bUseMaterialAttributes=true
  bUsedWithMeshDeformer=true
  bUsedWithSkeletalMesh=true
  blendMode=BLEND_Masked
  pixelDepthOffsetMode=PDOM_Legacy
  usedShadingModels=Default Lit
 }
 nodes [
  n1 = MaterialFunctionCall() {materialFunction=&MF_WheelParent_Simple}
  n2 = GetMaterialAttributes(None=@n1.MaterialAttributes) {attributeGetTypes [#1, #2]}
  n3 = TextureCoordinate() {coordinateIndex=1}
  n4 = AppendVector(
   A=Multiply(A=ComponentMask(None=@n3) {r=true}, B=ScalarParameter() {defaultValue=1, group=5_Special Effects, parameterName=GlowTileX, sortPriority=2})
   B=Multiply(A=ComponentMask(None=@n3) {g=true}, B=ScalarParameter() {defaultValue=1, group=5_Special Effects, parameterName=GlowTileY, sortPriority=3})
  )
  n5 = Multiply(A=Time(), B=ScalarParameter() {defaultValue=1, group=5_Special Effects, parameterName=GlowPanSpeedAdjust, sortPriority=4})
  n6 = Add(
   A=TextureSample(UVs=Panner(Coordinate=@n4, Time=@n5) {speedX=0.075, speedY=0.067}) {texture=/VehicleCosmetics/Wheels/Secretooze/Textures/T_Caustics_D}.G
   B=TextureSample(UVs=Panner(Coordinate=@n4, Time=@n5) {speedX=0.125, speedY=1}) {texture=&T_Caustics_D}.R
  )
  n7 = ScalarParameter() {defaultValue=5, group=5_Special Effects, parameterName=EmissivePattern}
  n8 = NamedRerouteUsage()
  n9 = ScalarParameter() {defaultValue=0.5, group=5_Special Effects, parameterName=EmissiveBase}
  n10 = TextureSample(UVs=@n3) {samplerType=SAMPLERTYPE_Grayscale, texture=/VehicleCosmetics/Wheels/_SharedTextures/FX/GradientCircle01}
  n11 = Add(A=Multiply(A=OneMinus(None=@n10.RGB)) {constB=4}, B=Multiply(B=Time()) {constA=-0.5})
  n12 = Add(A=Add(A=Multiply(A=@n6, B=@n7), B=Multiply(A=@n8, B=@n9)), B=Multiply(A=Frac(None=@n11), B=NamedRerouteUsage()))
  n13 = NamedRerouteDeclaration(None=Saturate(None=Multiply(A=@n12, B=@n12))) {name=FX_output, nodeColor=<rgba 0 0.8359 1 1>}
  n14 = Multiply(A=ComponentMask(None=@n1.RimRGBAMask) {g=true}, B=OneMinus(None=@n1.VertexColorR))
  n15 = LinearInterpolate(A=@n2.Base Color, B=@n13, Alpha=@n14) {constAlpha=1}
  n16 = LinearInterpolate(A=@n2.Emissive Color, B=@n13, Alpha=@n14)
  MaterialFunctionCall(In=MaterialFunctionCall(Material=SetMaterialAttributes!(MaterialAttributes=@n2.MaterialAttributes, Base Color=@n15, Emissive Color=@n16) {attributeSetTypes [#1, #2]}) {materialFunction=&MF_BaseVehicle_Damage}.Result) {materialFunction=&MF_VehicleDeRez_Transition}
  NamedRerouteDeclaration(None=Multiply(A=@n2.Base Color, B=@n14)) {name=RimGMask_Color, nodeColor=<rgba 0 0.2031 1 1>}
 ]
}
