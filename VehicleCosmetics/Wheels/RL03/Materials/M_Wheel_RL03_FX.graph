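# Translucent, skeletal-mesh-ready effect material for the RL03 wheel cosmetic, producing an animated emissive glow with a flame color.
# Emissive is a flat color parameter times an intensity, so all animation lives in opacity.
# Two caustics samples pan at different speeds over tileable UVs, combining squared green with red.
# A static gradient texture masks that sum and adds a faint base opacity; both textures come from other wheels' folders.
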Material /VehicleCosmetics/Wheels/RL03/Materials/M_Wheel_RL03_FX
graph {
 properties {
  bAutomaticallySetUsageInEditor=false
  bUseMaterialAttributes=true
  bUsedWithMeshDeformer=true
  bUsedWithSkeletalMesh=true
  blendMode=BLEND_Translucent
  pixelDepthOffsetMode=PDOM_Legacy
  usedShadingModels=Default Lit
 }
 nodes [
  n1 = VectorParameter() {group=0_SpecialEffects, parameterName=EmissiveFlameColor, sortPriority=0, defaultValue=<rgba 1 0.19 0 1>}
  n2 = TextureCoordinate()
  n3 = AppendVector(
   A=Multiply(A=ComponentMask(None=@n2) {r=true}, B=ScalarParameter() {defaultValue=1, group=0_SpecialEffects, parameterName=GlowTileX, sortPriority=2})
   B=Multiply(A=ComponentMask(None=@n2) {g=true}, B=ScalarParameter() {defaultValue=0.2, group=0_SpecialEffects, parameterName=GlowTileY, sortPriority=3})
  )
  n4 = Multiply(A=Time(), B=ScalarParameter() {defaultValue=0.1, group=0_SpecialEffects, parameterName=GlowPanSpeedAdjust, sortPriority=4})
  n5 = TextureSample(UVs=Panner(Coordinate=@n3, Time=@n4) {speedX=0.075, speedY=0.067}) {texture=/VehicleCosmetics/Wheels/Secretooze/Textures/T_Caustics_D}
  n6 = Add(
   A=Multiply(A=@n5.G, B=@n5.G)
   B=TextureSample(UVs=Panner(Coordinate=@n3, Time=@n4) {speedX=0.125, speedY=1}) {texture=&T_Caustics_D}.R
  )
  n7 = TextureSample() {samplerType=SAMPLERTYPE_LinearColor, texture=/VehicleCosmetics/Wheels/Twist_Twist/Textures/Gradient_V2_D}
  MakeMaterialAttributes!(
   BaseColor
   Metallic
   Specular
   Roughness
   Anisotropy
   EmissiveColor=Multiply(A=@n1.RGB, B=ScalarParameter() {defaultValue=32, group=0_SpecialEffects, parameterName=GlowEmissiveMultiply, sortPriority=1})
   Opacity=Add(A=Multiply(A=@n6, B=@n7.R), B=Multiply(A=@n7.R) {constB=0.05})
   OpacityMask
   Normal
   Tangent
   WorldPositionOffset
   SubsurfaceColor
   ClearCoat
   ClearCoatRoughness
   AmbientOcclusion
   Refraction
   CustomizedUVs_0
   CustomizedUVs_1
   CustomizedUVs_2
   CustomizedUVs_3
   CustomizedUVs_4
   CustomizedUVs_5
   CustomizedUVs_6
   CustomizedUVs_7
   PixelDepthOffset
   ShadingModel
   Displacement
  )
 ]
}
